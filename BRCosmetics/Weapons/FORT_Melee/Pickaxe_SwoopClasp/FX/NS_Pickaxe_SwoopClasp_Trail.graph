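# Niagara anim-trail effect for the SwoopClasp pickaxe swing, with a lightning ribbon and a motion-blur ribbon.
# Both emitters spawn FN anim trails between bone or socket pairs picked by User.SwingSectionIndex, using short-lived particles.
# The Lightning emitter draws two ribbon renderers, switches its color curves on User.Variant and fades on per-section system-age curves.
# Time-of-day scalars adjust brightness, and TrailMotionBlur also adjusts opacity for the front end.
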NiagaraSystem /BRCosmetics/Weapons/FORT_Melee/Pickaxe_SwoopClasp/FX/NS_Pickaxe_SwoopClasp_Trail
dependencies {
 usedDataInterfaces [
  &NiagaraDataInterfaceArrayInt32
  &NiagaraDataInterfaceCurve
  &NiagaraDataInterfaceColorCurve
  &NiagaraDataInterfaceSocketReader
 ]
 usedDynamicInputs []
 usedModules [
  &self:SystemSpawnScript.SystemScriptSource.SystemScriptGraph.NiagaraNodeAssignment_1.SetVariables_#1
  &self:SystemSpawnScript.SystemScriptSource.SystemScriptGraph.NiagaraNodeAssignment_0.SetVariables_#2
  &self:SystemSpawnScript.SystemScriptSource.SystemScriptGraph.NiagaraNodeAssignment_2.SetVariables_#3
  &SystemState
 ]
}
emitters {
 Lightning {
  emitterSpawnScript [&FN_RibbonIDOffsetUtility()]
  emitterUpdateScript [
   &EmitterState()
   &FN_SpawnAnimTrail()
  ]
  particleSpawnScript [
   &FN_InitializeAnimTrailBoneIndecies(
    First Bone or Socket Index:NiagaraInt32=&SelectIntFromArray(Direct Array Index=$User.SwingSectionIndex, Int Selection Array=DataInterface() {propertyValues {IntData [2,0,2,0]}})
    Second Bone or Socket Index:NiagaraInt32=&SelectIntFromArray(Direct Array Index=$User.SwingSectionIndex, Int Selection Array=DataInterface() {propertyValues {IntData [3,1,3,1]}})
   )
   &FN_InitializeAnimTrail(AnimTrailWidth=2.5, Color=<rgba 10 10 10 1>, Lifetime=0.25)
  ]
  particleUpdateScript [
   &FN_UpdateAnimTrail() {module=FN_UpdateAnimTrail001}
   /SetVariables(Particles.TrailWidth=1) {isSetParameters=true}
   /SetVariables(Particles.DY2:Vector4f=&MakeVector4(X=1, Y=1, Z=1, W=1)) {isSetParameters=true}
   /SetVariables(Particles.Trail Color02:LinearColor=&MakeLinearColorFromVectorAndFloat(
     "Vector (RGB)"=&MakeVectorFromLinearColorRGB(Color=&MultiplyLinearColor(
       Linear Color=<rgba 1 1 1 0.9>
       Scale Factor=&MakeVector4_FromLinearColor(LinearColor=&MergeHelper_ColorSwitch(
         0=&ColorFromCurve(ColorCurve=DataInterface() {propertyValues {
            AlphaCurve {keys []}
            BlueCurve {keys [<at 0 1>,<at 0.5005 0.4573>]}
            GreenCurve {keys [<at 0 0.2609>,<at 0.5005 1>]}
            RedCurve {keys [<at 0 0>,<at 0.5005 0>]}
           }})
         1=&ColorFromCurve(ColorCurve=DataInterface() {propertyValues {AlphaCurve {keys []}, BlueCurve {keys [<at 0 1>]}, GreenCurve {keys [<at 0 0>]}, RedCurve {keys [<at 0 0.4583>]}}})
         2=?
         3=?
         Index=&MakeIntFromFloat(Float=$User.Variant)
         Num=::NewEnumerator0
        ))
      ))
     "Float (Alpha)"=&Multiply_Float(
      A=1
      B=&FloatFromCurve(
       FloatCurve=DataInterface() {propertyValues {Curve {keys [
           {arriveTangent=-0.9541, interpMode=RCIM_Cubic, leaveTangent=-0.9541, tangentMode=RCTM_User, time=0, value=1}
           {arriveTangent=-0.5324, interpMode=RCIM_Cubic, leaveTangent=-0.5324, tangentMode=RCTM_User, time=0.5, value=0}
          ]}}}
       Scale Curve=&MergeHelper_FloatSwitch(
        0=&FloatFromCurve(
         FloatCurve=DataInterface() {propertyValues {Curve {keys [
             {arriveTangent=-0.8776, interpMode=RCIM_Cubic, leaveTangent=-0.8776, tangentMode=RCTM_User, time=0, value=1}
             {arriveTangent=-1.5205, tangentMode=RCTM_User, time=0.18, value=0}
            ]}}}
         CurveIndex=$Engine.System.Age
        )
        1=&FloatFromCurve(
         FloatCurve=DataInterface() {propertyValues {Curve {keys [
             {arriveTangent=-0.8776, interpMode=RCIM_Cubic, leaveTangent=-0.8776, tangentMode=RCTM_User, time=0.1, value=1}
             {arriveTangent=-1.5205, tangentMode=RCTM_User, time=0.22, value=0}
            ]}}}
         CurveIndex=$Engine.System.Age
        )
        2=&FloatFromCurve(
         FloatCurve=DataInterface() {propertyValues {Curve {keys [
             {arriveTangent=-0.8776, interpMode=RCIM_Cubic, leaveTangent=-0.8776, tangentMode=RCTM_User, time=0.1, value=1}
             {arriveTangent=-1.5205, tangentMode=RCTM_User, time=0.22, value=0}
            ]}}}
         CurveIndex=$Engine.System.Age
        )
        3=&FloatFromCurve(
         FloatCurve=DataInterface() {propertyValues {Curve {keys [
             {arriveTangent=-0.8776, interpMode=RCIM_Cubic, leaveTangent=-0.8776, tangentMode=RCTM_User, time=0.1, value=1}
             {arriveTangent=-1.5205, tangentMode=RCTM_User, time=0.22, value=0}
            ]}}}
         CurveIndex=$Engine.System.Age
        )
        Index=$User.SwingSectionIndex
        Num=::NewEnumerator2
       )
      )
     )
    )) {isSetParameters=true}
   &DynamicMaterialParameters(
    Index 0 Param 1=10
    Index 0 Param 2=&FloatFromCurve(
     FloatCurve=DataInterface() {propertyValues {Curve {keys [
         {arriveTangent=-3.0186, interpMode=RCIM_Cubic, leaveTangent=-3.0186, tangentMode=RCTM_User, time=0, value=1}
         {interpMode=RCIM_Cubic, tangentMode=RCTM_User, time=1, value=0}
        ]}}}
     Scale Curve=0.8
    )
    Module.Param0WriteEnabled=true
    Module.Param1WriteEnabled=true
    Module.Param2WriteEnabled=false
    Module.Param3WriteEnabled=false
   )
   &ScaleColor(
    Scale Alpha=&FloatFromCurve(
     FloatCurve=DataInterface() {propertyValues {Curve {keys [{arriveTangent=-0.2884, interpMode=RCIM_Cubic, leaveTangent=-0.2884, tangentMode=RCTM_User, time=0, value=1}, <at 0.5 0>]}}}
     Scale Curve=&MergeHelper_FloatSwitch(
      0=&FloatFromCurve(
       FloatCurve=DataInterface() {propertyValues {Curve {keys [
           {arriveTangent=-0.8776, interpMode=RCIM_Cubic, leaveTangent=-0.8776, tangentMode=RCTM_User, time=0, value=1}
           {arriveTangent=-1.5205, tangentMode=RCTM_User, time=0.18, value=0}
          ]}}}
       CurveIndex=$Engine.System.Age
      )
      1=&FloatFromCurve(
       FloatCurve=DataInterface() {propertyValues {Curve {keys [
           {arriveTangent=-0.8776, interpMode=RCIM_Cubic, leaveTangent=-0.8776, tangentMode=RCTM_User, time=0.1, value=1}
           {arriveTangent=-1.5205, tangentMode=RCTM_User, time=0.22, value=0}
          ]}}}
       CurveIndex=$Engine.System.Age
      )
      2=&FloatFromCurve(
       FloatCurve=DataInterface() {propertyValues {Curve {keys [
           {arriveTangent=-0.8776, interpMode=RCIM_Cubic, leaveTangent=-0.8776, tangentMode=RCTM_User, time=0.1, value=1}
           {arriveTangent=-1.5205, tangentMode=RCTM_User, time=0.22, value=0}
          ]}}}
       CurveIndex=$Engine.System.Age
      )
      3=&FloatFromCurve(
       FloatCurve=DataInterface() {propertyValues {Curve {keys [
           {arriveTangent=-0.8776, interpMode=RCIM_Cubic, leaveTangent=-0.8776, tangentMode=RCTM_User, time=0.1, value=1}
           {arriveTangent=-1.5205, tangentMode=RCTM_User, time=0.22, value=0}
          ]}}}
       CurveIndex=$Engine.System.Age
      )
      Index=$User.SwingSectionIndex
      Num=::NewEnumerator2
     )
    )
    Scale RGB=&Multiply_VectorByFloat(
     Vector=&MakeVectorFromLinearColorRGB(Color=&MergeHelper_ColorSwitch(
       0=&ColorFromCurve(ColorCurve=DataInterface() {propertyValues {
          AlphaCurve {keys []}
          BlueCurve {keys [<at 0 1>,<at 0.5005 0.4573>]}
          GreenCurve {keys [<at 0 0.2609>,<at 0.5005 1>]}
          RedCurve {keys [<at 0 0>,<at 0.5005 0>]}
         }})
       1=&ColorFromCurve(ColorCurve=DataInterface() {propertyValues {
          AlphaCurve {keys []}
          BlueCurve {keys [<at 0 1>,<at 0.7884 1>]}
          GreenCurve {keys [<at 0 0>,<at 0.7884 0>]}
          RedCurve {keys [<at 0 0.1667>,<at 0.7884 0.25>]}
         }})
       2=?
       3=?
       Index=&MakeIntFromFloat(Float=$User.Variant)
       Num=::NewEnumerator0
      ))
     Float=&Multiply_Float(A=25, B=&FNTimeOfDayScalar(Night Scale=0.5, Morning Scale=1, Evening Scale=1))
    )
    ScaleA=true
    ScaleRGB=true
   )
  ]
  rendererClasses [&NiagaraRibbonRendererProperties]
  renderers [
   {rendererClass=&NiagaraRibbonRendererProperties}
   {rendererClass=&NiagaraRibbonRendererProperties, rendererIndex=1}
  ]
 }
 TrailMotionBlur {
  emitterSpawnScript [&FN_RibbonIDOffsetUtility()]
  emitterUpdateScript [
   &EmitterState()
   &FN_SpawnAnimTrail(AnimTrailWidthMode=::FromSecond)
  ]
  particleSpawnScript [
   &FN_InitializeAnimTrailBoneIndecies(
    First Bone or Socket Index:NiagaraInt32=&SelectIntFromArray(Direct Array Index=$User.SwingSectionIndex, Int Selection Array=DataInterface() {propertyValues {IntData [7,5,7,5]}})
    Second Bone or Socket Index:NiagaraInt32=&SelectIntFromArray(Direct Array Index=$User.SwingSectionIndex, Int Selection Array=DataInterface() {propertyValues {IntData [6,4,6,4]}})
   )
   &FN_InitializeAnimTrail(Color=<rgba 0.3084 0.4452 0.5625 1>, Lifetime=0.25)
  ]
  particleUpdateScript [
   &SolveForcesAndVelocity(Clamp Velocity=false, Limit Acceleration=false)
   &FN_UpdateAnimTrail() {module=FN_UpdateAnimTrail001}
   &ScaleColor(
    Scale Alpha=&FloatFromCurve(
     FloatCurve=DataInterface() {propertyValues {Curve {keys [{interpMode=RCIM_Cubic, tangentMode=RCTM_User, time=0, value=1}, {interpMode=RCIM_Cubic, tangentMode=RCTM_User, time=0.5, value=0}]}}}
     Scale Curve=&FloatFromCurve(
      FloatCurve=DataInterface() {propertyValues {Curve {keys [<at 0.1 1 RCIM_Cubic>,<at 0.25 0 RCIM_Cubic>]}}}
      CurveIndex=$Emitter.Age
      Scale Curve=&Multiply_Float(
       A=&FN_FrontEndFloatModifier(
        Gameplay=&FNTimeOfDayScalar(Night Scale=0.4, Morning Scale=0.3, Day Scale=0.5, Evening Scale=0.3)
        Front End=0.25
       )
       B=&FloatFromCurve(
        FloatCurve=DataInterface() {propertyValues {Curve {keys [
            {arriveTangent=-0.8776, interpMode=RCIM_Cubic, leaveTangent=-0.8776, tangentMode=RCTM_User, time=0.1, value=1}
            {arriveTangent=-1.5205, tangentMode=RCTM_User, time=0.22, value=0}
           ]}}}
        CurveIndex=$Engine.System.Age
        Scale Curve=0.75
       )
      )
     )
    )
    Scale RGB=&MakeVectorFromLinearColorRGB(Color=$System.ToDColor)
    ScaleA=true
    ScaleRGB=true
   )
   &ScaleRibbonWidth()
  ]
  rendererClasses [&NiagaraRibbonRendererProperties]
  renderers [{rendererClass=&NiagaraRibbonRendererProperties}]
 }
}
summary {
 systemName=NS_Pickaxe_SwoopClasp_Trail
 emitters [
  {emitterName=TrailMotionBlur, rendererClasses [&NiagaraRibbonRendererProperties]}
  {emitterName=Lightning, rendererClasses [&NiagaraRibbonRendererProperties]}
 ]
 userVariables [
  User.bIsFrontEnd {type=NiagaraBool}
  User.bIsFrontEndPreview {type=NiagaraBool}
  User.RibbonIDOffset {type=NiagaraInt32}
  User.SwingSectionIndex {type=NiagaraInt32}
  User.Variant {type=NiagaraFloat, defaultValue=0, description=""}
 ]
}
